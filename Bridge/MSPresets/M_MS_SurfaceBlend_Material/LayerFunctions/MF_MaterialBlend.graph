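# Library material function for the Megascans surface blend material that blends Material A and Material B with a height-based mask driven by vertex color.
# It mixes the two input heights by Blend Amount, scales by Blend Power with optional inversion, then offsets the result and adds doubled vertex color (red by default).
# CheapContrast sharpens that mask by Blend Falloff.
# Outputs are the blended attributes, their red WorldDisplacement as Height, and the Alpha.
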MaterialFunction /Bridge/MSPresets/M_MS_SurfaceBlend_Material/LayerFunctions/MF_MaterialBlend
graph {
 properties {bExposeToLibrary=true}
 nodes [
  n1 = LinearInterpolate(
   A=Saturate(None=FunctionInput() {bUsePreviewValueAsDefault=true, inputName=A Height, inputType=FunctionInput_Scalar, sortPriority=3})
   B=FunctionInput(Preview=Constant()) {bUsePreviewValueAsDefault=true, inputName=B Height, inputType=FunctionInput_Scalar, sortPriority=5}
   Alpha=FunctionInput() {bUsePreviewValueAsDefault=true, inputName=Blend Amount, inputType=FunctionInput_Scalar, sortPriority=6}
  )
  n2 = FunctionInput() {bUsePreviewValueAsDefault=true, inputName=Blend Power, inputType=FunctionInput_Scalar, sortPriority=9}
  n3 = OneMinus(None=@n1)
  n4 = Subtract(A=LinearInterpolate(
    A=Multiply(A=@n1, B=@n2)
    B=Multiply(A=@n3, B=@n2)
    Alpha=FunctionInput() {bUsePreviewValueAsDefault=true, inputName=Invert Blend, inputType=FunctionInput_Scalar, sortPriority=7}
   ))
  n5 = Multiply(A=FunctionInput(Preview=VertexColor().R) {bUsePreviewValueAsDefault=true, inputName=Vertex Color, inputType=FunctionInput_Scalar}) {constB=2}
  n6 = MaterialFunctionCall(
   In=Saturate(None=Add(A=@n4, B=@n5))
   Contrast=FunctionInput() {bUsePreviewValueAsDefault=true, inputName=Blend Falloff, inputType=FunctionInput_Scalar, sortPriority=8}
  ) {materialFunction=&CheapContrast}
  FunctionOutput(None=@n6.Result) {outputName=Alpha, sortPriority=2}
  n7 = BlendMaterialAttributes(
   A=FunctionInput(Preview=SetMaterialAttributes!(MaterialAttributes, WorldDisplacement) {attributeSetTypes [#1]}) {bUsePreviewValueAsDefault=true, inputName=Material A, inputType=FunctionInput_MaterialAttributes, sortPriority=2}
   B=FunctionInput() {bUsePreviewValueAsDefault=true, inputName=Material B, inputType=FunctionInput_MaterialAttributes, sortPriority=4}
   Alpha=@n6.Result
  )
  FunctionOutput(None=@n7) {outputName=MA}
  FunctionOutput(None=ComponentMask(None=GetMaterialAttributes(None=@n7) {attributeGetTypes [#1]}.WorldDisplacement) {r=true}) {outputName=Height, sortPriority=1}
 ]
}
