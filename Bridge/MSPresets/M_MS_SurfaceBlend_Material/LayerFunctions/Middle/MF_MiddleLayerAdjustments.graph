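# Material function that adjusts the middle layer's material attributes in the Megascans surface blend material.
# Base color gets saturation, brightness, tint and contrast (power) controls from parameters in the "03 - Middle Layer" group.
# A static switch sets specular from either a default value or half the saturated dot product of the original albedo and desaturation weights.
# Input roughness interpolates between min and max parameters, and the result leaves as material attributes.
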MaterialFunction /Bridge/MSPresets/M_MS_SurfaceBlend_Material/LayerFunctions/Middle/MF_MiddleLayerAdjustments
graph {nodes [
  n1 = GetMaterialAttributes(None=FunctionInput() {inputType=FunctionInput_MaterialAttributes}) {attributeGetTypes [
    #1
    #2
    #3
    #4
    #5
   ]}
  n2 = VectorParameter() {
   group=03 - Middle Layer
   parameterName=Middle Layer Albedo Controls
   sortPriority=9
   channelNames {a=Useless, b=Contrast, g=Brightness, r=Saturation}
   defaultValue=<rgba 1 1 1 1>
  }
  n3 = Multiply(A=Desaturation(None=@n1.BaseColor, Fraction=OneMinus(None=@n2.Saturation)), B=@n2.Brightness)
  n4 = VectorParameter() {group=03 - Middle Layer, parameterName=Middle Layer Albedo Tint, sortPriority=10, defaultValue=<rgba 1 1 1 1>}
  n5 = VectorParameter() {group=03 - Middle Layer, parameterName=Middle Layer Specular Desaturation, sortPriority=13, defaultValue=<rgba 0.5 0.5 0.5 1>}
  n6 = Saturate(None=DotProduct(A=@n1.BaseColor, B=@n5.RGB))
  n7 = SetMaterialAttributes!(
   MaterialAttributes=@n1.MaterialAttributes
   BaseColor=Power(Base=Multiply(A=@n3, B=@n4.RGB), Exp=@n2.Contrast)
   Specular=StaticSwitchParameter(
    True=Multiply(A=@n6) {constB=0.5}
    False=ScalarParameter() {defaultValue=0.5, group=03 - Middle Layer, parameterName=Middle Layer Specular Default, sortPriority=12}
   ) {group=03 - Middle Layer, parameterName="Middle Layer  Specular from Albedo", sortPriority=11}
   Roughness=LinearInterpolate(
    A=ScalarParameter() {group=03 - Middle Layer, parameterName=Middle Layer Roughness Min, sortPriority=14}
    B=ScalarParameter() {defaultValue=1, group=03 - Middle Layer, parameterName=Middle Layer Roughness Max, sortPriority=15}
    Alpha=ComponentMask(None=@n1.Roughness) {r=true}
   )
   WorldDisplacement=@n1.MaterialAttributes
   Normal=@n1.MaterialAttributes
   TessellationMultiplier=@n1.MaterialAttributes
  ) {attributeSetTypes [
    #1
    #6
    #2
    #3
    #4
    #5
   ]}
  FunctionOutput(None=@n7) {outputName=Output MA}
 ]}
