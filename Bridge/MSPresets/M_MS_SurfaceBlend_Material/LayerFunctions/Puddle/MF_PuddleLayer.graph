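# Material layer function that adds puddles with wet, darkened edges to incoming material attributes, driven by a vertex color input and displacement height.
# A darkening mask tints albedo and adjusts roughness, while a liquid mask blends in color, roughness, flattened AO and animated four-way chaos wave normals.
# View-angle Fresnel scales liquid color opacity.
# The output rebuilds only base color, specular, roughness, normal and AO, dropping other attributes like displacement.
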MaterialFunction /Bridge/MSPresets/M_MS_SurfaceBlend_Material/LayerFunctions/Puddle/MF_PuddleLayer
graph {
 properties {bExposeToLibrary=true}
 nodes [
  n1 = GetMaterialAttributes(None=FunctionInput() {inputName=MA, inputType=FunctionInput_MaterialAttributes, sortPriority=1}) {attributeGetTypes [
    #1
    #2
    #3
    #4
    #5
    #6
    #7
    #8
   ]}
  n2 = VectorParameter() {group=05 - Puddle Layer, parameterName=Albedo Darkening Color, sortPriority=8, defaultValue=<rgba 0.3281 0.3281 0.3281 1>}
  n3 = ComponentMask(None=@n1.WorldDisplacement) {r=true}
  n4 = FunctionInput() {inputName=Vertex Color, inputType=FunctionInput_Scalar}
  n5 = ScalarParameter() {defaultValue=-0.75, group=05 - Puddle Layer, parameterName=Albedo Darkening Range, sliderMax=-0.9, sortPriority=9}
  n6 = Saturate(None=Multiply(A=Add(A=@n4, B=@n5) {constB=-0.5}) {constB=4})
  n7 = Saturate(None=Add(
    A=Subtract(A=@n3, B=@n6)
    B=ScalarParameter() {defaultValue=0.03, group=05 - Puddle Layer, parameterName=Albedo Darkening Falloff, sortPriority=10}
   ))
  n8 = LinearInterpolate(A=@n1.BaseColor, B=Multiply(A=@n1.BaseColor, B=@n2.RGB), Alpha=@n7)
  n9 = VectorParameter() {group=05 - Puddle Layer, parameterName=Liquid Color, defaultValue=<rgba 0.1213 0.1675 0.1979 1>}
  n10 = LinearInterpolate(
   B=ScalarParameter() {defaultValue=1, group=05 - Puddle Layer, parameterName=Liquid Height Max, sortPriority=1}
   Alpha=ComponentMask(None=@n1.WorldDisplacement) {r=true}
  )
  n11 = Constant() {r=1}
  n12 = Subtract(A=@n4, B=Saturate(None=Add(A=@n10, B=@n11)))
  n13 = Constant() {r=-0.5}
  n14 = ComponentMask(None=Saturate(None=Divide(A=@n12, B=@n13))) {r=true}
  n15 = ScalarParameter() {defaultValue=0.25, group=05 - Puddle Layer, parameterName=Liquid Opacity, sortPriority=1}
  n16 = DotProduct(A=CameraVectorWS(), B=VertexNormalWS())
  n17 = Multiply(
   A=Saturate(None=Multiply(A=@n14, B=@n15))
   B=Power(
    Base=Saturate(None=@n16)
    Exp=ScalarParameter() {defaultValue=0.25, group=05 - Puddle Layer, parameterName=Liquid Fresnel Power, sortPriority=2}
   )
  )
  n18 = LinearInterpolate(
   A=@n1.Roughness
   B=ScalarParameter() {defaultValue=0.3, group=05 - Puddle Layer, parameterName=Liquid Falloff Roughness, sortPriority=7}
   Alpha=@n7
  )
  n19 = ScalarParameter() {defaultValue=0.01, group=05 - Puddle Layer, parameterName=Liquid Roughness, sortPriority=6}
  n20 = VectorParameter() {
   group=05 - Puddle Layer
   parameterName=Liquid Wave Controls
   sortPriority=12
   channelNames {a=Useless, b=Normal Strength, g=Wave Speed, r=Wave Scale}
   defaultValue=<rgba 1 0.005 0.05 0>
  }
  n21 = TextureObjectParameter() {
   group=05 - Puddle Layer
   parameterName=Liquid Wave Normal
   samplerType=SAMPLERTYPE_Normal
   sortPriority=11
   texture=/Engine/Engine_MI_Shaders/T_Base_Tile_Normal
  }
  n22 = MaterialFunctionCall!(Coordinates=Multiply(A=TextureCoordinate(), B=@n20.Wave Scale), Speed=@n20.Wave Speed, Divisor, Texture=@n21) {materialFunction=/Engine/Functions/Engine_MaterialFunctions03/ComplexMotion/Motion_4WayChaos_Normal}
  n23 = Multiply(A=ComponentMask(None=@n22.Result) {g=true, r=true}, B=@n20.Normal Strength)
  n24 = ComponentMask(None=@n22.Result) {b=true}
  n25 = LinearInterpolate(A=@n1.Normal, B=Normalize(VectorInput=AppendVector(A=@n23, B=@n24)), Alpha=@n14)
  n26 = Constant() {r=1}
  FunctionOutput(None=MakeMaterialAttributes!(
    BaseColor=LinearInterpolate(A=@n8, B=@n9.RGB, Alpha=@n17)
    Metallic
    Specular=@n1.Specular
    Roughness=LinearInterpolate(A=@n18, B=@n19, Alpha=@n14)
    Anisotropy
    EmissiveColor
    Opacity
    OpacityMask
    Normal=Normalize(VectorInput=@n25)
    Tangent
    WorldPositionOffset
    SubsurfaceColor
    ClearCoat
    ClearCoatRoughness
    AmbientOcclusion=LinearInterpolate(A=@n1.AmbientOcclusion, B=@n26, Alpha=@n14)
    Refraction
    CustomizedUVs_0
    CustomizedUVs_1
    CustomizedUVs_2
    CustomizedUVs_3
    CustomizedUVs_4
    CustomizedUVs_5
    CustomizedUVs_6
    CustomizedUVs_7
    PixelDepthOffset
    ShadingModel
    Displacement
   )) {outputName=MA}
 ]
}
