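# Megascans surface blend master material that layers base, middle and top textured surfaces with optional scatter and puddle layers through vertex painting.
# Material attributes chain through height-aware blend functions: red mixes middle with base, green adds top, alpha adds scatter, blue drives puddles.
# Static switches toggle each layer's adjustments plus the scatter and puddle stages.
# Vector parameters with named channels control the base/middle, top and scatter blends.
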Material /Bridge/MSPresets/M_MS_SurfaceBlend_Material/M_MS_SurfaceBlend_Material
graph {
 properties {bUseMaterialAttributes=true, pixelDepthOffsetMode=PDOM_Legacy, usedShadingModels=Default Lit}
 nodes [
  n1 = VertexColor()
  n2 = MaterialFunctionCall() {materialFunction=/Bridge/MSPresets/M_MS_SurfaceBlend_Material/LayerFunctions/MF_ScatterLayer}
  n3 = MaterialFunctionCall() {materialFunction=/Bridge/MSPresets/M_MS_SurfaceBlend_Material/LayerFunctions/Top/MF_TopLayerTextures}
  n4 = MaterialFunctionCall() {materialFunction=/Bridge/MSPresets/M_MS_SurfaceBlend_Material/LayerFunctions/Middle/MF_MiddleLayerTextures}
  n5 = MaterialFunctionCall() {materialFunction=/Bridge/MSPresets/M_MS_SurfaceBlend_Material/LayerFunctions/Base/MF_BaseLayerTextures}
  n6 = VectorParameter() {
   group=01 - Blend Controls
   parameterName=Base/Middle Layer Blend Controls
   channelNames {a=Invert Blend, b=Blend Falloff, g=Blend Contrast, r=Blend Amount}
   defaultValue=<rgba 1 1 1 0>
  }
  n7 = MaterialFunctionCall!(
   Vertex Color=@n1.R
   Material A=StaticSwitchParameter(
    True=MaterialFunctionCall!(In=@n4.MA) {materialFunction=/Bridge/MSPresets/M_MS_SurfaceBlend_Material/LayerFunctions/Middle/MF_MiddleLayerAdjustments}.Output MA
    False=@n4.MA
   ) {group=00 - Global, parameterName=Use Middle Layer Adjustments, sortPriority=3}
   A Height=@n4.Height
   Material B=StaticSwitchParameter(
    True=MaterialFunctionCall!(In=@n5.MA) {materialFunction=/Bridge/MSPresets/M_MS_SurfaceBlend_Material/LayerFunctions/Base/MF_BaseLayerAdjustments}.Output MA
    False=@n5.MA
   ) {group=00 - Global, parameterName=Use Base Layer Adjustments, sortPriority=2}
   B Height=@n5.Height
   Blend Amount=@n6.Blend Amount
   Invert Blend=@n6.Blend Amount
   Blend Falloff=@n6.Blend Amount
   Blend Power=@n6.Blend Amount
  ) {materialFunction=/Bridge/MSPresets/M_MS_SurfaceBlend_Material/LayerFunctions/MF_MaterialBlend}
  n8 = VectorParameter() {
   group=01 - Blend Controls
   parameterName=Top Layer Blend Controls
   sortPriority=1
   channelNames {a=Invert Blend, b=Blend Falloff, g=Blend Contrast, r=Blend Amount}
   defaultValue=<rgba 1 1 1 0>
  }
  n9 = MaterialFunctionCall(
   Vertex Color=@n1.G
   Material A=StaticSwitchParameter(
    True=MaterialFunctionCall!(In=@n3.MA) {materialFunction=/Bridge/MSPresets/M_MS_SurfaceBlend_Material/LayerFunctions/Top/MF_TopLayerAdjustments}.Output MA
    False=@n3.MA
   ) {group=00 - Global, parameterName=Use Top Layer Adjustments, sortPriority=4}
   A Height=@n3.Height
   Material B=@n7.MA
   B Height=@n7.MA
   Blend Amount=@n8.Blend Amount
   Invert Blend=@n8.Blend Amount
   Blend Falloff=@n8.Blend Amount
   Blend Power=@n8.Blend Amount
  ) {materialFunction=&MF_MaterialBlend}
  n10 = VectorParameter() {
   group=01 - Blend Controls
   parameterName=Scatter Layer Blend Controls
   sortPriority=2
   channelNames {a=Invert Blend, b=Blend Falloff, g=Blend Contrast, r=Blend Amount}
   defaultValue=<rgba 1 1 1 0>
  }
  n11 = StaticSwitchParameter(
   True=MaterialFunctionCall!(
    Vertex Color=@n1.A
    Material A=@n2.MA
    A Height=@n2.MA
    Material B=@n9.MA
    B Height=@n9.MA
    Opacity=@n2.MA
    Blend Amount=@n10.Blend Amount
    Invert Blend=@n10.Blend Amount
    Blend Falloff=@n10.Blend Amount
    Blend Power=@n10.Blend Amount
   ) {materialFunction=/Bridge/MSPresets/M_MS_SurfaceBlend_Material/LayerFunctions/MF_BlendScatter}.MA
   False=@n9.MA
  ) {group=00 - Global, parameterName=Use Scatter Layer, sortPriority=1}
  StaticSwitchParameter(
   True=MaterialFunctionCall!(Vertex Color=@n1.B, MA=@n11) {materialFunction=/Bridge/MSPresets/M_MS_SurfaceBlend_Material/LayerFunctions/Puddle/MF_PuddleLayer}.MA
   False=@n11
  ) {group=00 - Global, parameterName=Use Puddle Layer}
 ]
}
